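# Physics asset for the jacket of the female medium Cyber Scavenger body, with capsules on pelvis, spine and thighs as body shapes.
# Seven two-bone chains of box bodies around the front, sides and back are constrained to spine_04, with generally tighter swing limits at the back.
# Angular-locked cross constraints join neighboring lower segments from each front panel around to the back center, tying the hem together while leaving the front open.
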PhysicsAsset /BRCosmetics/Characters/Player/Female/Medium/Bodies/F_MED_Cyber_Scavenger/Meshes/F_MED_Cyber_Scavenger_Physics
bodies {
 dyn_jacket_bk_01 [{shapeType=Box, extentX=5, extentY=10, extentZ=2, center=<2.6674 0 0.3948>}]
 dyn_jacket_bk_02 [{shapeType=Box, extentX=6.51, extentY=11.01, extentZ=3.01, center=<3.2457 0 0.3946>}]
 dyn_jacket_bk_l_01 [{shapeType=Box, extentX=4.5, extentY=8, extentZ=1.5, center=<1.7409 0.0223 0.2586>, rotation=<rot -1.7639 -0.6441 21.589>}]
 dyn_jacket_bk_l_02 [{shapeType=Box, extentX=7.51, extentY=10.01, extentZ=2.01, center=<2.741 -0.6208 -0.8764>, rotation=<rot 3.0274 -7.5447 21.3747>}]
 dyn_jacket_bk_r_01 [{shapeType=Box, extentX=6.5, extentY=8, extentZ=1.5, center=<-3.3413 -0.1938 -0.8751>, rotation=<rot -0.0002 -0.0007 15.0041>}]
 dyn_jacket_bk_r_02 [{shapeType=Box, extentX=6.51, extentY=9.01, extentZ=2.01, center=<-3.2768 0.4812 0.5002>, rotation=<rot -7.0915 -12.2328 14.8949>}]
 dyn_jacket_fr_l_01 [{shapeType=Box, extentX=7, extentY=7.5, extentZ=2, center=<3.5707 -3.3342 0.5695>, rotation=<rot 3.4054 -9.409 19.7236>}]
 dyn_jacket_fr_l_02 [{shapeType=Box, extentX=9, extentY=8, extentZ=1.5, center=<3.5725 -5.5749 0.0585>, rotation=<rot -9.8466 -10.1526 1.7538>}]
 dyn_jacket_fr_r_01 [{shapeType=Box, extentX=7.5, extentY=7.5, extentZ=2, center=<-3.4571 4.4225 -1.1556>, rotation=<rot 0.1723 -0.1301 8.2797>}]
 dyn_jacket_fr_r_02 [{shapeType=Box, extentX=8.51, extentY=7.01, extentZ=2.01, center=<-3.9248 4.7664 -0.1263>, rotation=<rot -3.5884 -3.6113 -1.0684>}]
 dyn_jacket_l_01 [{shapeType=Box, extentX=5.5, extentY=8, extentZ=2, center=<2.6801 -0.1076 -0.6114>, rotation=<rot 1.7285 -9.8524 9.8556>}]
 dyn_jacket_l_02 [{shapeType=Box, extentX=8.5, extentY=7.5, extentZ=2, center=<3.8583 -0.6806 -1.619>, rotation=<rot 0 -10.0043 0>}]
 dyn_jacket_r_01 [{shapeType=Box, extentX=5.5, extentY=8, extentZ=1.5, center=<-2.611 -0.035 0.3499>, rotation=<rot 0 -0.0005 5.0018>}]
 dyn_jacket_r_02 [{shapeType=Box, extentX=9.01, extentY=8.51, extentZ=1.51, center=<-3.7881 0.5685 1.6214>, rotation=<rot -5.8144 -12.88 11.535>}]
 pelvis [{shapeType=Capsule, length=10.3035, radius=10.3795, center=<-0.1856 -1.5184 -0.1972>, rotation=<rot 5.5723 -90 0.5389>}]
 spine_01 [{shapeType=Capsule, length=6.948, radius=8.7038, center=<-1.0334 1.1831 0.0196>, rotation=<rot 0.7761 -89.9999 -0.0409>}]
 spine_02 [{shapeType=Capsule, length=5.4062, radius=7.9396, center=<-0.4183 1.9186 -0.8613>, rotation=<rot 0 -90 0>}]
 spine_04 [{shapeType=Capsule, length=1.75, radius=10.2192, center=<-3.0072 2.9175 0.1822>, rotation=<rot -4.8294 -90 -0.5872>}]
 thigh_l [{shapeType=Capsule, length=16.4627, radius=8.0068, center=<-14.6231 0.3316 -0.4294>, rotation=<rot 0.0656 -92.9029 -90.7503>}]
 thigh_r [{shapeType=Capsule, length=16.2127, radius=8.2568, center=<14.6425 -0.1506 0.2624>, rotation=<rot 0.0656 -92.9028 -90.7494>}]
}
constraints [
 {bone1Name=dyn_jacket_fr_l_01, bone2Name=spine_04, swing1LimitDegrees=20, swing2LimitDegrees=20, twistLimitDegrees=5}
 {bone1Name=dyn_jacket_fr_l_02, bone2Name=dyn_jacket_fr_l_01, swing1LimitDegrees=30, swing2LimitDegrees=15, twistLimitDegrees=5}
 {bone1Name=dyn_jacket_fr_r_01, bone2Name=spine_04, swing1LimitDegrees=20, swing2LimitDegrees=20, twistLimitDegrees=5}
 {bone1Name=dyn_jacket_fr_r_02, bone2Name=dyn_jacket_fr_r_01, swing1LimitDegrees=30, swing2LimitDegrees=15, twistLimitDegrees=5}
 {bone1Name=dyn_jacket_l_01, bone2Name=spine_04, swing1LimitDegrees=15, swing2LimitDegrees=15, twistLimitDegrees=5}
 {bone1Name=dyn_jacket_l_02, bone2Name=dyn_jacket_l_01, swing1LimitDegrees=40, swing2LimitDegrees=15, twistLimitDegrees=5}
 {bone1Name=dyn_jacket_r_01, bone2Name=spine_04, swing1LimitDegrees=15, swing2LimitDegrees=15, twistLimitDegrees=5}
 {bone1Name=dyn_jacket_r_02, bone2Name=dyn_jacket_r_01, swing1LimitDegrees=40, swing2LimitDegrees=15, twistLimitDegrees=5}
 {bone1Name=dyn_jacket_bk_l_01, bone2Name=spine_04, swing1LimitDegrees=5, swing2LimitDegrees=5, twistLimitDegrees=5}
 {bone1Name=dyn_jacket_bk_l_02, bone2Name=dyn_jacket_bk_l_01, swing1LimitDegrees=20, swing2LimitDegrees=5, twistLimitDegrees=5}
 {bone1Name=dyn_jacket_bk_r_01, bone2Name=spine_04, swing1LimitDegrees=5, swing2LimitDegrees=5, twistLimitDegrees=5}
 {bone1Name=dyn_jacket_bk_r_02, bone2Name=dyn_jacket_bk_r_01, swing1LimitDegrees=20, swing2LimitDegrees=5, twistLimitDegrees=5}
 {bone1Name=dyn_jacket_bk_01, bone2Name=spine_04, swing1LimitDegrees=5, swing2LimitDegrees=5, twistLimitDegrees=5}
 {bone1Name=dyn_jacket_bk_02, bone2Name=dyn_jacket_bk_01, swing1LimitDegrees=5, swing2LimitDegrees=5, twistLimitDegrees=5}
 {
  bone1Name=dyn_jacket_l_02
  bone2Name=dyn_jacket_fr_l_02
  swing1LimitDegrees=45
  swing1Motion=Locked
  swing2LimitDegrees=45
  swing2Motion=Locked
  twistLimitDegrees=45
  twistMotion=Locked
 }
 {
  bone1Name=dyn_jacket_bk_l_02
  bone2Name=dyn_jacket_l_02
  swing1LimitDegrees=45
  swing1Motion=Locked
  swing2LimitDegrees=45
  swing2Motion=Locked
  twistLimitDegrees=45
  twistMotion=Locked
 }
 {
  bone1Name=dyn_jacket_bk_02
  bone2Name=dyn_jacket_bk_l_02
  swing1LimitDegrees=45
  swing1Motion=Locked
  swing2LimitDegrees=45
  swing2Motion=Locked
  twistLimitDegrees=45
  twistMotion=Locked
 }
 {
  bone1Name=dyn_jacket_r_02
  bone2Name=dyn_jacket_fr_r_02
  swing1LimitDegrees=45
  swing1Motion=Locked
  swing2LimitDegrees=45
  swing2Motion=Locked
  twistLimitDegrees=45
  twistMotion=Locked
 }
 {
  bone1Name=dyn_jacket_bk_r_02
  bone2Name=dyn_jacket_r_02
  swing1LimitDegrees=45
  swing1Motion=Locked
  swing2LimitDegrees=45
  swing2Motion=Locked
  twistLimitDegrees=45
  twistMotion=Locked
 }
 {
  bone1Name=dyn_jacket_bk_02
  bone2Name=dyn_jacket_bk_r_02
  swing1LimitDegrees=45
  swing1Motion=Locked
  swing2LimitDegrees=45
  swing2Motion=Locked
  twistLimitDegrees=45
  twistMotion=Locked
 }
]
