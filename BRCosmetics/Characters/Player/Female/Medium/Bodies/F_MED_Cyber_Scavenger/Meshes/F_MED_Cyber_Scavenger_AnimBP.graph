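# Post-copy animation Blueprint for the Cyber Scavenger female body cosmetic, adding collar and jacket secondary motion.
# It copies the parent mesh pose, then offsets collar bones using alphas mapped from head, clavicle and arm rotations.
# RigidBody simulates that cached pose; a backpack check on customization load selects a variant pulling back jacket bones toward the pelvis with LookAt.
# A spine-driven jacket rotation chain exists but is unconnected.
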AnimBlueprint /BRCosmetics/Characters/Player/Female/Medium/Bodies/F_MED_Cyber_Scavenger/Meshes/F_MED_Cyber_Scavenger_AnimBP
target_skeleton=&/F_MED_Cyber_Scavenger_Skeleton
graphs [
 AnimGraph {nodes [
   n1 = Get jacketBackRot() -> (jacketBackRot:Rotator)
   n2 = "Transform (Modify) Bone - Bone: dyn_jacket_bk_01"(Alpha=1, Translation="0,0,0", Rotation=@n1, Scale="1,1,1") {
    rotationMode=BMM_Replace
    rotationSpace=BCS_BoneSpace
    scaleMode=BMM_Replace
    scaleSpace=BCS_BoneSpace
    translationMode=BMM_Additive
    translationSpace=BCS_BoneSpace
    alphaBoolBlend {blendOption=Linear}
    boneToModify {boneName=dyn_jacket_bk_01}
   }
   n3 = "Transform (Modify) Bone - Bone: dyn_jacket_bk_r_01"(ComponentPose=@n2, Alpha=1, Translation="0,0,0", Rotation=@n1, Scale="1,1,1") {
    rotationMode=BMM_Replace
    rotationSpace=BCS_BoneSpace
    scaleMode=BMM_Replace
    scaleSpace=BCS_BoneSpace
    translationMode=BMM_Additive
    translationSpace=BCS_BoneSpace
    alphaBoolBlend {blendOption=Linear}
    boneToModify {boneName=dyn_jacket_bk_r_01}
   }
   "Transform (Modify) Bone - Bone: dyn_jacket_bk_l_01"(ComponentPose=@n3, Alpha=1, Translation="0,0,0", Rotation=@n1, Scale="1,1,1") {
    rotationMode=BMM_Replace
    rotationSpace=BCS_BoneSpace
    scaleMode=BMM_Replace
    scaleSpace=BCS_BoneSpace
    translationMode=BMM_Additive
    translationSpace=BCS_BoneSpace
    alphaBoolBlend {blendOption=Linear}
    boneToModify {boneName=dyn_jacket_bk_l_01}
   }
   Output Pose(Result=Blend Poses by bool(
     BlendPose_0=Use cached pose 'backpack_ON_CopyPoseCache'()
     BlendPose_1=Use cached pose 'backpack_OFF_CopyPoseCache'()
     bActiveValue=Get IsBackpackOn() -> (IsBackpackOn:bool)
    ) {blendType=Linear}) {layerGroup=None}
   n4 = RigidBody(ComponentPose=Local To Component(LocalPose=Use cached pose 'CopyPoseBaseCache'())) {
    bClampLinearTranslationLimitToRefPose=true
    bOverrideWorldGravity=true
    overridePhysicsAsset=&/F_MED_Cyber_Scavenger_Physics
    simulationSpace=BaseBoneSpace
    alphaBoolBlend {blendOption=Linear}
    baseBoneRef {boneName=spine_01}
    componentAppliedLinearAccClamp=<2000 2000 2000>
    componentLinearAccScale=<0.5 0.5 0.5>
    componentLinearVelScale=<0.5 0.5 0.5>
    overrideWorldGravity {z=-980}
   }
   n5 = Look At - Bone: dyn_jacket_bk_01(ComponentPose=@n4, Alpha=0.4) {
    bUseLookUpAxis=true
    alphaBoolBlend {blendOption=Linear}
    boneToModify {boneName=dyn_jacket_bk_01}
    lookAtLocation {x=0}
    lookAtTarget {boneReference {boneName=pelvis}}
    lookAt_Axis {axis=<1 0>}
    lookUp_Axis {axis=<0 1>}
   }
   n6 = Look At - Bone: dyn_jacket_bk_l_01(ComponentPose=@n5, Alpha=0.4) {
    bUseLookUpAxis=true
    alphaBoolBlend {blendOption=Linear}
    boneToModify {boneName=dyn_jacket_bk_l_01}
    lookAtLocation {x=0}
    lookAtTarget {boneReference {boneName=pelvis}}
    lookAt_Axis {axis=<1 0>}
    lookUp_Axis {axis=<0 1>}
   }
   n7 = Look At - Bone: dyn_jacket_bk_r_01(ComponentPose=@n6, Alpha=0.4) {
    bUseLookUpAxis=true
    alphaBoolBlend {blendOption=Linear}
    boneToModify {boneName=dyn_jacket_bk_r_01}
    lookAtLocation {x=0}
    lookAtTarget {boneReference {boneName=pelvis}}
    lookAt_Axis {axis=<-1 0>}
    lookUp_Axis {axis=<0 1>}
   }
   Save cached pose 'backpack_ON_CopyPoseCache'(Pose=Component To Local(ComponentPose=@n7)) {cacheName=backpack_ON_CopyPoseCache}
   n8 = "Transform (Modify) Bone - Bone: dyn_collar_bk"(
    ComponentPose=Local To Component(LocalPose=Copy Pose From Mesh() {bUseAttachedParent=true})
    Alpha=Get headRotBack() -> (headRotBack:double)
    Translation="0,0,0"
    Rotation="0,0,-25"
    Scale="1,1,1"
   ) {
    rotationMode=BMM_Replace
    rotationSpace=BCS_BoneSpace
    scaleMode=BMM_Replace
    scaleSpace=BCS_BoneSpace
    translationMode=BMM_Additive
    translationSpace=BCS_BoneSpace
    alphaBoolBlend {blendOption=Linear}
    boneToModify {boneName=dyn_collar_bk}
    rotation {roll=-25}
   }
   n9 = "Transform (Modify) Bone - Bone: dyn_collar_l_md"(ComponentPose=@n8, Alpha=Get clavicleRotUp_l() -> (clavicleRotUp_l:double), Translation="-1,0,2", Rotation="0,0,0", Scale="1,1,1") {
    rotationMode=BMM_Replace
    rotationSpace=BCS_BoneSpace
    scaleMode=BMM_Replace
    scaleSpace=BCS_BoneSpace
    translationMode=BMM_Additive
    translationSpace=BCS_BoneSpace
    alphaBoolBlend {blendOption=Linear}
    boneToModify {boneName=dyn_collar_l_md}
    translation {x=-1, z=2}
   }
   n10 = Get armRotUp_r() -> (armRotUp_r:double)
   n11 = "Transform (Modify) Bone - Bone: dyn_collar_r_fr"(ComponentPose=@n9, Alpha=@n10, Translation="4,0,-1", Rotation="5,0,0", Scale="1,1,1") {
    rotationMode=BMM_Replace
    rotationSpace=BCS_BoneSpace
    scaleMode=BMM_Replace
    scaleSpace=BCS_BoneSpace
    translationMode=BMM_Additive
    translationSpace=BCS_BoneSpace
    alphaBoolBlend {blendOption=Linear}
    boneToModify {boneName=dyn_collar_r_fr}
    rotation {pitch=5}
    translation {x=4, z=-1}
   }
   n12 = "Transform (Modify) Bone - Bone: dyn_collar_r_md"(ComponentPose=@n11, Alpha=@n10, Translation="2,0,-1", Rotation="3,0,0", Scale="1,1,1") {
    rotationMode=BMM_Replace
    rotationSpace=BCS_BoneSpace
    scaleMode=BMM_Replace
    scaleSpace=BCS_BoneSpace
    translationMode=BMM_Additive
    translationSpace=BCS_BoneSpace
    alphaBoolBlend {blendOption=Linear}
    boneToModify {boneName=dyn_collar_r_md}
    rotation {pitch=3}
    translation {x=2, z=-1}
   }
   Save cached pose 'CopyPoseBaseCache'(Pose=Component To Local(ComponentPose=@n12)) {cacheName=CopyPoseBaseCache}
   n13 = RigidBody(ComponentPose=Local To Component(LocalPose=Use cached pose 'CopyPoseBaseCache'())) {
    bClampLinearTranslationLimitToRefPose=true
    bOverrideWorldGravity=true
    overridePhysicsAsset=&F_MED_Cyber_Scavenger_Physics
    simulationSpace=BaseBoneSpace
    alphaBoolBlend {blendOption=Linear}
    baseBoneRef {boneName=spine_01}
    componentAppliedLinearAccClamp=<2000 2000 2000>
    componentLinearAccScale=<0.5 0.5 0.5>
    componentLinearVelScale=<0.5 0.5 0.5>
    overrideWorldGravity {z=-980}
   }
   Save cached pose 'backpack_OFF_CopyPoseCache'(Pose=Component To Local(ComponentPose=@n13)) {cacheName=backpack_OFF_CopyPoseCache}
  ]}
 EventGraph {nodes [
   Event OnCharacterCustomizationLoaded()
   n1 = Get playerPawn() -> (playerPawn:Fort Player Pawn Athena Object Reference)
   > Is Valid(InputObject=@n1)
   > Do Once()
   n2 = CheckForBackbling()
   > Set IsBackpackOn(IsBackpackOn:bool=OR Boolean(A=@n2, B=@n2.bIsBackblingLong)) -> (Output_Get:bool)
   Event BlueprintUpdateAnimation()
   > n3 = Sequence!(execute) -> (then_0, then_1, then_2:Exec, then_3:Exec, then_4:Exec)
   n4 = GetOwnerSkeletalMeshComponent()
   n5 = BreakTransform(InTransform=GetDeltaTransformFromRefPose(self=@n4, BoneName=clavicle_l, BaseName=spine_05))
   > Set clavicleRotUp_l(clavicleRotUp_l:double=MapRangeClamped(Value=BreakRotator(InRot=@n5.Rotation).Pitch, InRangeA=0, InRangeB=45)) -> (Output_Get:double)
   n6 = BreakTransform(InTransform=GetDeltaTransformFromRefPose(self=@n4, BoneName=upperarm_r, BaseName=spine_05))
   Set armRotUp_r(execute=@n3.then_1, armRotUp_r:double=MapRangeClamped(Value=BreakRotator(InRot=@n6.Rotation).Yaw, InRangeA=60, InRangeB=130)) -> (Output_Get:double)
   n7 = BreakTransform(InTransform=GetDeltaTransformFromRefPose(self=@n4, BoneName=head, BaseName=spine_05))
   Set headRotBack(
    execute=@n3.then_2
    headRotBack:double=MapRangeClamped(Value=BreakRotator(InRot=@n7.Rotation).Yaw, InRangeA=-40, InRangeB=-25, OutRangeA=1, OutRangeB=0)
   ) -> (Output_Get:double)
   Event BlueprintInitializeAnimation()
   n8 = TryGetPawnOwner()
   > Is Valid(InputObject=@n8)
   > n9 = Cast To FortPlayerPawnAthena(Object=@n8)
   > Set playerPawn(playerPawn:Fort Player Pawn Athena Object Reference=@n9) -> (Output_Get:Fort Player Pawn Athena Object Reference)
   n10 = BreakRotator(InRot=BreakTransform(InTransform=GetDeltaTransformFromRefPose(self=@n4, BoneName=spine_04, BaseName=spine_02)).Rotation)
   n11 = Branch(Condition="float >= float"(A=@n10.Yaw, B=0) {is=title K2Node_CallFunction})
   > Set jacketBackRot(jacketBackRot:Rotator=MakeRotator(Pitch=float * float(A=@n10.Yaw, B=-0.5))) -> (Output_Get:Rotator)
   Set jacketBackRot(execute=@n11.else, jacketBackRot="0, 0, 0")
   "Not Equal (Object)"!(A:Object Reference=GetSkeletalMeshForPartType(self=@n1, PartType=Backpack), B:Object Reference) -> (ReturnValue:bool)
  ]}
]
variables [
 armRotUp_r {type=number}
 clavicleRotUp_l {type=number}
 headRotBack {type=number}
 IsBackpackOn {type=boolean}
 jacketBackRot {type=Rotator}
 playerPawn {type=FortPlayerPawnAthena}
]
