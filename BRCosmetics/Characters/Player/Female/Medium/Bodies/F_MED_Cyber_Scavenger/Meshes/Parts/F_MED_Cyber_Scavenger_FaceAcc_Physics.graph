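# Physics asset for the F_MED_Cyber_Scavenger face accessory mesh, setting up hair bone physics for this female medium player body.
# Capsules on the dyn_hair bones form short strands constrained to the head, plus a two-bone b chain and a three-bone main chain.
# Most strands have tight swing limits, while the main chain swings much wider.
# A head sphere and four spine_05 boxes provide body shapes, linked by an angularly free constraint.
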PhysicsAsset /BRCosmetics/Characters/Player/Female/Medium/Bodies/F_MED_Cyber_Scavenger/Meshes/Parts/F_MED_Cyber_Scavenger_FaceAcc_Physics
bodies {
 dyn_hair_a_01 [{shapeType=Capsule, length=3.9745, radius=0.8185, center=<0.6113 0 0>, rotation=<rot -90 -0.0396 0.0397>}]
 dyn_hair_b_01 [{shapeType=Capsule, length=2.5, radius=1.5, center=<0.2976 0 0.026>, rotation=<rot -84.9997 0.0002 -0.0002>}]
 dyn_hair_b_02 [{shapeType=Capsule, length=1.8133, radius=1.0738, center=<0.2781 0 0>, rotation=<rot -90 10.025 -10.025>}]
 dyn_hair_l_a_01 [{shapeType=Capsule, length=1.8167, radius=0.75, center=<0.2549 -0.0231 0.0683>, rotation=<rot -74.9997 -0.0001 -5.0001>}]
 dyn_hair_l_b_01 [{shapeType=Capsule, length=2.7251, radius=0.5871, center=<0.917 0 0>, rotation=<rot 90 0.0066 0.0066>}]
 dyn_hair_l_c_01 [{shapeType=Capsule, length=2.7635, radius=0.876, center=<0.6147 0 0.0538>, rotation=<rot -85 0 0>}]
 dyn_hair_main_01 [{shapeType=Capsule, length=3.1876, radius=3.2372, center=<0.8569 0 0.695>, rotation=<rot -79.9985 0 84.9991>}]
 dyn_hair_main_02 [{shapeType=Capsule, length=3.9451, radius=3.8889, center=<1.1965 0.0344 0.3925>, rotation=<rot 0 0 84.9996>}]
 dyn_hair_main_03 [{shapeType=Capsule, length=7.5304, radius=1.7905, center=<1.0141 -0.0712 0.407>, rotation=<rot -1.2986 -75.2897 -82.6131>}]
 dyn_hair_r_a_01 [{shapeType=Capsule, length=1.5366, radius=0.7465, center=<0.2356 0.0007 0.1577>, rotation=<rot 71.4374 -138.2302 -143.485>}]
 dyn_hair_r_b_01 [{shapeType=Capsule, length=2.0441, radius=0.75, center=<0.6026 0 0.1063>, rotation=<rot -79.9995 0.0001 -0.0001>}]
 head [{shapeType=Sphere, radius=8.5819, center=<6.5152 0.1818 -0.0457>}]
 spine_05 [
  {shapeType=Box, extentX=1.01, extentY=7.51, extentZ=7.51, center=<14.188 -7.549 0>, rotation=<rot 0 -90 0>}
  {shapeType=Box, extentX=1.01, extentY=7.51, extentZ=10.01, center=<14.3733 -3.9769 -6.526>, rotation=<rot -50.0001 -90.0001 0>}
  {shapeType=Box, extentX=1.01, extentY=7.51, extentZ=9.51, center=<14.1341 -3.7561 6.088>, rotation=<rot 60.0001 -90.0001 0>}
  {shapeType=Box, extentX=1.5, extentY=10.5, extentZ=7.5, center=<16.1567 -2.2979 0>}
 ]
}
constraints [
 {bone1Name=dyn_hair_b_01, bone2Name=head, swing1LimitDegrees=10, swing2LimitDegrees=10, twistLimitDegrees=10}
 {bone1Name=dyn_hair_b_02, bone2Name=dyn_hair_b_01, swing1LimitDegrees=20, swing2LimitDegrees=20, twistLimitDegrees=10}
 {bone1Name=dyn_hair_r_a_01, bone2Name=head, swing1LimitDegrees=15, swing2LimitDegrees=15, twistLimitDegrees=10}
 {bone1Name=dyn_hair_a_01, bone2Name=head, swing1LimitDegrees=10, swing2LimitDegrees=10, twistLimitDegrees=10}
 {bone1Name=dyn_hair_l_a_01, bone2Name=head, swing1LimitDegrees=10, swing2LimitDegrees=10, twistLimitDegrees=10}
 {bone1Name=dyn_hair_main_01, bone2Name=head, swing1LimitDegrees=50, swing2LimitDegrees=50, twistLimitDegrees=10}
 {bone1Name=dyn_hair_main_02, bone2Name=dyn_hair_main_01, swing1LimitDegrees=60, swing2LimitDegrees=60, twistLimitDegrees=10}
 {bone1Name=dyn_hair_main_03, bone2Name=dyn_hair_main_02, swing1LimitDegrees=40, swing2LimitDegrees=40, twistLimitDegrees=10}
 {bone1Name=dyn_hair_l_c_01, bone2Name=head, swing1LimitDegrees=10, swing2LimitDegrees=10, twistLimitDegrees=10}
 {bone1Name=dyn_hair_r_b_01, bone2Name=head, swing1LimitDegrees=15, swing2LimitDegrees=15, twistLimitDegrees=10}
 {bone1Name=dyn_hair_l_b_01, bone2Name=head, swing1LimitDegrees=15, swing2LimitDegrees=15, twistLimitDegrees=10}
 {
  bone1Name=head
  bone2Name=spine_05
  swing1LimitDegrees=45
  swing1Motion=Free
  swing2LimitDegrees=45
  swing2Motion=Free
  twistLimitDegrees=45
  twistMotion=Free
 }
]
