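# Post-process-style AnimBP for the female Cyber Scavenger's face accessory mesh, driving dynamic hair from the attached parent's pose.
# Each update maps head yaw relative to spine_05 into headRotBackClamp, which blends additive pitch on two hair bones.
# A RigidBody node then simulates physics in spine_05 space with overridden gravity.
# A hairClipping sine value feeds three hair-scaling nodes, but that chain is disconnected from the output.
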AnimBlueprint /BRCosmetics/Characters/Player/Female/Medium/Bodies/F_MED_Cyber_Scavenger/Meshes/Parts/F_MED_Cyber_Scavenger_FaceAcc_AnimBP
target_skeleton=&/F_MED_Cyber_Scavenger_FaceAcc_Skeleton
graphs [
 AnimGraph {nodes [
   n1 = Get hairClipping() -> (hairClipping:double)
   n2 = "Transform (Modify) Bone - Bone: dyn_hair_main_01"(Alpha=@n1, Translation="0,0,0", Rotation="0,0,0", Scale="0.2,0.2,0.2") {
    rotationMode=BMM_Additive
    rotationSpace=BCS_BoneSpace
    scaleMode=BMM_Additive
    scaleSpace=BCS_BoneSpace
    translationMode=BMM_Additive
    translationSpace=BCS_BoneSpace
    alphaBoolBlend {blendOption=Linear}
    boneToModify {boneName=dyn_hair_main_01}
    scale=<0.2 0.2 0.2>
   }
   n3 = "Transform (Modify) Bone - Bone: dyn_hair_main_02"(ComponentPose=@n2, Alpha=@n1, Translation="0,0,0", Rotation="0,0,0", Scale="0.2,0.2,0.2") {
    rotationMode=BMM_Additive
    rotationSpace=BCS_BoneSpace
    scaleMode=BMM_Additive
    scaleSpace=BCS_BoneSpace
    translationMode=BMM_Additive
    translationSpace=BCS_BoneSpace
    alphaBoolBlend {blendOption=Linear}
    boneToModify {boneName=dyn_hair_main_02}
    scale=<0.2 0.2 0.2>
   }
   "Transform (Modify) Bone - Bone: dyn_hair_main_03"(ComponentPose=@n3, Alpha=@n1, Translation="0,0,0", Rotation="0,0,0", Scale="0.2,0.2,0.21") {
    rotationMode=BMM_Additive
    rotationSpace=BCS_BoneSpace
    scaleMode=BMM_Additive
    scaleSpace=BCS_BoneSpace
    translationMode=BMM_Additive
    translationSpace=BCS_BoneSpace
    alphaBoolBlend {blendOption=Linear}
    boneToModify {boneName=dyn_hair_main_03}
    scale=<0.2 0.2 0.21>
   }
   n4 = Get headRotBackClamp() -> (headRotBackClamp:double)
   n5 = "Transform (Modify) Bone - Bone: dyn_hair_main_01"(
    ComponentPose=Local To Component(LocalPose=Copy Pose From Mesh() {bUseAttachedParent=true})
    Alpha=@n4
    Translation="0,0,0"
    Rotation="15,0,0"
    Scale="1,1,1"
   ) {
    rotationMode=BMM_Additive
    rotationSpace=BCS_BoneSpace
    scaleMode=BMM_Additive
    scaleSpace=BCS_BoneSpace
    translationMode=BMM_Additive
    translationSpace=BCS_BoneSpace
    alphaBoolBlend {blendOption=Linear}
    boneToModify {boneName=dyn_hair_main_01}
    rotation {pitch=15}
   }
   n6 = "Transform (Modify) Bone - Bone: dyn_hair_main_02"(ComponentPose=@n5, Alpha=@n4, Translation="0,0,0", Rotation="35,0,0", Scale="1,1,1") {
    rotationMode=BMM_Additive
    rotationSpace=BCS_BoneSpace
    scaleMode=BMM_Additive
    scaleSpace=BCS_BoneSpace
    translationMode=BMM_Additive
    translationSpace=BCS_BoneSpace
    alphaBoolBlend {blendOption=Linear}
    boneToModify {boneName=dyn_hair_main_02}
    rotation {pitch=35}
   }
   n7 = RigidBody(ComponentPose=@n6) {
    bClampLinearTranslationLimitToRefPose=true
    bOverrideWorldGravity=true
    overridePhysicsAsset=&/F_MED_Cyber_Scavenger_FaceAcc_Physics
    simulationSpace=BaseBoneSpace
    alphaBoolBlend {blendOption=Linear}
    baseBoneRef {boneName=spine_05}
    componentAppliedLinearAccClamp=<1000 1000 1000>
    componentLinearAccScale=<1 1 1>
    componentLinearVelScale=<1 1 1>
    overrideWorldGravity {z=-980}
   }
   Output Pose(Result=Component To Local(ComponentPose=@n7)) {layerGroup=None}
  ]}
 EventGraph {nodes [
   Event BlueprintUpdateAnimation()
   > n1 = Sequence!(execute) -> (then_0, then_1, then_2:Exec, then_3:Exec)
   n2 = GetOwnerSkeletalMeshComponent()
   n3 = BreakRotator(InRot=BreakTransform(InTransform=GetDeltaTransformFromRefPose(self=@n2, BoneName=head, BaseName=spine_05)).Rotation)
   > Set headRotBackClamp(headRotBackClamp:double=MapRangeClamped(Value=@n3.Yaw, InRangeA=-30, InRangeB=0, OutRangeA=1, OutRangeB=0)) -> (Output_Get:double)
   Set hairClipping(
    execute=@n1.then_1
    hairClipping:double="Sin (Degrees)"!(A:double=MapRangeClamped(Value=@n3.Yaw, InRangeA=-30, InRangeB=0, OutRangeA=180, OutRangeB=0)) -> (ReturnValue:double)
   ) -> (Output_Get:double)
   TryGetPawnOwner()
  ]}
]
variables [
 hairClipping {type=number}
 headRotBackClamp {type=number}
]
